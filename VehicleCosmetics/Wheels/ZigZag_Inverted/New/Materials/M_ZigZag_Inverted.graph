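# Lit material for the ZigZag Inverted vehicle wheel, with animated emissive rim and tread effects and a Painted switch for its colors.
# Panned noise, channel-packed mask and VFX textures, and vertex color channels supply named reroutes for the rim circles, tread and Fresnel-boosted emissive.
# A facet normal distorts the VFX lookup, whose blue channel plus Time drives hard-edged repeating tread pulses.
# The resulting attributes pass through shared vehicle damage and de-rez transition functions.
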Material /VehicleCosmetics/Wheels/ZigZag_Inverted/New/Materials/M_ZigZag_Inverted
graph {
 properties {
  bAutomaticallySetUsageInEditor=false
  bUseMaterialAttributes=true
  bUsedWithMeshDeformer=true
  bUsedWithSkeletalMesh=true
  usedShadingModels=Default Lit
 }
 nodes [
  n1 = NamedRerouteUsage()
  n2 = SetMaterialAttributes!(
   MaterialAttributes
   Emissive Color=NamedRerouteUsage()
   Normal=LinearInterpolate(A=NamedRerouteUsage(), B=TextureSample() {samplerType=SAMPLERTYPE_Normal, texture=/VehicleCosmetics/Textures/T_BlankNormal_B}.RGB, Alpha=@n1)
  ) {attributeSetTypes [#1, #2]}
  GetMaterialAttributes(None=MaterialFunctionCall(In=MaterialFunctionCall(Material=@n2) {materialFunction=&MF_BaseVehicle_Damage}.Result) {materialFunction=&MF_VehicleDeRez_Transition}.Result) {attributeGetTypes [#3]}
  n3 = TextureCoordinate() {uTiling=0.5, vTiling=0.5}
  n4 = Multiply(A=Time()) {constB=0.3}
  n5 = Panner(Coordinate=@n3, Time=@n4) {speedX=0.1, speedY=0.1}
  n6 = ConstantBiasScale(None=OneMinus(None=TextureSample(UVs=@n5) {samplerSource=SSM_Wrap_WorldGroupSettings, texture=/VehicleCosmetics/Wheels/_SharedTextures/FX/alpine_infinite_NoiseClouds}.R)) {bias=-0.65, scale=3}
  n7 = Panner(Coordinate=@n3, Time=@n4) {speedX=-0.1, speedY=0.05}
  n8 = ConstantBiasScale(None=OneMinus(None=TextureSample(UVs=@n7) {samplerSource=SSM_Wrap_WorldGroupSettings, texture=&alpine_infinite_NoiseClouds}.R)) {bias=-0.7, scale=3}
  NamedRerouteDeclaration(None=Add(A=Clamp(None=@n6), B=Clamp(None=@n8))) {name=Swirl Noise, nodeColor=<rgba 1 0 0.0703 1>}
  n9 = TextureCoordinate() {coordinateIndex=1}
  n10 = TextureCoordinate() {coordinateIndex=1, uTiling=2, vTiling=2}
  n11 = Add(
   A=TextureSample(UVs=Panner(Coordinate=@n9) {speedX=0.1, speedY=0.1}) {samplerSource=SSM_Wrap_WorldGroupSettings, texture=/VehicleCosmetics/Wheels/_SharedTextures/FX/CloudNoise_01_uncompressed}.R
   B=TextureSample(UVs=Panner(Coordinate=@n10) {speedX=-0.1, speedY=0.05}) {samplerSource=SSM_Wrap_WorldGroupSettings, texture=&CloudNoise_01_uncompressed}.R
  )
  NamedRerouteDeclaration(None=Clamp(None=ConstantBiasScale(None=@n11) {bias=-0.1, scale=2})) {name=Cloud Noise Rim, nodeColor=<rgba 0 0.3906 1 1>}
  n12 = TextureSampleParameter2D() {
   group=00_Base
   parameterName=VFX
   samplerSource=SSM_Wrap_WorldGroupSettings
   sortPriority=1
   texture=/VehicleCosmetics/Wheels/ZigZag_Inverted/New/Textures/T_ZigZagSE_Tire_RGB
  }
  NamedRerouteDeclaration(None=@n12.A) {name=VFX A, nodeColor=<rgba 1 1 1 1>}
  n13 = VectorParameter() {group=00_Base, parameterName=Paint Color, sortPriority=5, defaultValue=<rgba 1 0.0545 0 1>}
  NamedRerouteDeclaration(None=StaticSwitchParameter(True=@n13.RGB, False=VectorParameter() {group=00_Base, parameterName=Accent Color, sortPriority=4, defaultValue=<rgba 1 0 0.0284 1>}.RGB) {group=00_Base, parameterName=Painted, sortPriority=2}) {name=Accent Color, nodeColor=<rgba 1 0 0.3984 1>}
  NamedRerouteDeclaration(None=StaticSwitchParameter(True=@n13.RGB, False=VectorParameter() {group=00_Base, parameterName=Custom Color, sortPriority=3, defaultValue=<rgba 0.84 0.84 0.84 1>}.RGB) {group=00_Base, parameterName=Painted, sortPriority=2}) {name=Custom Color, nodeColor=<rgba 1 0 0.3984 1>}
  n14 = TextureSampleParameter2D() {
   group=00_Base
   parameterName=Masks
   samplerSource=SSM_Wrap_WorldGroupSettings
   samplerType=SAMPLERTYPE_LinearColor
   sortPriority=0
   texture=/VehicleCosmetics/Wheels/ZigZag_Inverted/New/Textures/T_ZigZagSE_Rim_RGB
  }
  NamedRerouteDeclaration(None=Add(
    A=NamedRerouteDeclaration(None=@n14.R) {name=Masks R, nodeColor=<rgba 1 0 0 1>}
    B=NamedRerouteDeclaration(None=@n14.B) {name=B_Channel, nodeColor=<rgba 0 0.0445 0.8278 1>}
   )) {name=BC, nodeColor=<rgba 1 0.9609 0 1>}
  n15 = NamedRerouteUsage()
  n16 = Power(Base=@n15) {constExponent=1.5}
  n17 = Clamp(None=Power(Base=@n15) {constExponent=8}) {maxDefault=2}
  n18 = NamedRerouteUsage()
  n19 = Add(A=Multiply(A=Clamp(None=@n16), B=NamedRerouteUsage()), B=Add(A=Multiply(A=@n17, B=@n18), B=@n17))
  n20 = NamedRerouteUsage()
  n21 = LinearInterpolate(A=NamedRerouteUsage(), B=Add(A=Multiply(A=@n19, B=@n20), B=Clamp(None=@n17) {maxDefault=0.3}), Alpha=NamedRerouteUsage())
  n22 = MaterialFunctionCall!(Normal, Exponent=Constant() {r=3}) {materialFunction=/VehicleCosmetics/Wheels/Cola_Holographic/Materials/MF_Fresnel_Classic}
  n23 = Multiply(A=NamedRerouteUsage(), B=@n22.Result)
  n24 = Multiply(A=Multiply(A=@n22.Result) {constB=0.5}, B=NamedRerouteUsage())
  n25 = NamedRerouteUsage()
  NamedRerouteDeclaration(None=LinearInterpolate(A=Add(A=@n21, B=@n23), B=Add(A=@n24, B=@n25), Alpha=@n1)) {name=Emissive, nodeColor=<rgba 0 1 0.1562 1>}
  n26 = VertexColor()
  NamedRerouteDeclaration(None=@n26.R) {name=VC R, nodeColor=<rgba 1 0 0 1>}
  n27 = Multiply(A=ScalarParameter() {defaultValue=2, group=00_Base, parameterName=Circles Swirl Intensity, sortPriority=20}, B=NamedRerouteUsage())
  NamedRerouteDeclaration(None=@n27) {name=Circles Swirl, nodeColor=<rgba 0 0.4844 1 1>}
  n28 = NamedRerouteUsage()
  n29 = Add(A=Multiply(A=@n27, B=@n28), B=@n28)
  n30 = NamedRerouteUsage()
  n31 = Add(
   A=Multiply(A=NamedRerouteUsage(), B=@n30)
   B=Multiply(A=@n30, B=ScalarParameter() {group=00_Base, parameterName=Circles BG Intensity, sortPriority=21})
  )
  n32 = Add(
   A=Multiply(A=@n29, B=ScalarParameter() {defaultValue=0.3, group=00_Base, parameterName=Circles Noise Intensity, sortPriority=22})
   B=Multiply(A=Add(A=@n29, B=@n31), B=NamedRerouteUsage())
  )
  n33 = ScalarParameter() {defaultValue=1, group=00_Base, parameterName=Circles Intensity, sortPriority=23}
  NamedRerouteDeclaration(None=Multiply(A=Multiply(A=@n32, B=@n33), B=ScalarParameter() {defaultValue=1, group=00_Base, parameterName=Rim Overal Intensity, sortPriority=24})) {name=RIm, nodeColor=<rgba 0.8047 1 0 1>}
  n34 = Multiply(A=NamedRerouteUsage(), B=NamedRerouteUsage())
  n35 = Clamp(None=@n34)
  n36 = NamedRerouteUsage()
  n37 = Multiply(A=Add(A=@n35, B=@n36), B=ScalarParameter() {defaultValue=2, group=00_Base, parameterName=Tread Line Intensity, sortPriority=30})
  n38 = NamedRerouteUsage()
  n39 = LinearInterpolate(A=@n38, B=@n38, Alpha=NamedRerouteUsage())
  n40 = ScalarParameter() {defaultValue=1, group=00_Base, parameterName=Tread Line 2 Intensity, sortPriority=31}
  n41 = Add(A=Multiply(A=Multiply(A=@n37, B=@n35), B=@n39), B=Multiply(A=Multiply(A=@n36, B=@n40), B=@n34))
  n42 = NamedRerouteUsage()
  n43 = NamedRerouteUsage()
  n44 = Multiply(A=Multiply(A=NamedRerouteUsage(), B=@n42), B=@n43)
  n45 = TextureSample() {samplerSource=SSM_Wrap_WorldGroupSettings, samplerType=SAMPLERTYPE_Normal, texture=/VehicleCosmetics/Wheels/_SharedTextures/FX/T_facet2_N}
  n46 = ComponentMask(None=@n45.RGB) {g=true, r=true}
  n47 = TextureSampleParameter2D(UVs=Add(A=Divide(A=@n46) {constB=20}, B=TextureCoordinate())) {
   group=00_Base
   parameterName=VFX
   samplerSource=SSM_Wrap_WorldGroupSettings
   sortPriority=1
   texture=&T_ZigZagSE_Tire_RGB
  }
  n48 = Add(A=@n47.B, B=Time())
  n49 = Constant() {r=1}
  n50 = If(A=Clamp(None=Frac(None=@n48)) {maxDefault=0.5}, "A > B"=Constant(), "A == B"=@n49, "A < B"=@n49) {constB=0.19}
  n51 = NamedRerouteUsage()
  n52 = Clamp(None=Add(A=@n51, B=NamedRerouteUsage()))
  n53 = Add(A=Multiply(A=Multiply(A=@n43) {constB=0.2}, B=@n42), B=Multiply(A=Multiply(A=@n50, B=@n52), B=OneMinus(None=@n51)))
  n54 = Multiply(
   A=Multiply(A=@n39, B=Add(A=@n44, B=@n53))
   B=ScalarParameter() {defaultValue=2, group=00_Base, parameterName=Tread Intensity, sortPriority=33}
  )
  NamedRerouteDeclaration(None=Multiply(A=Add(A=@n41, B=@n54), B=ScalarParameter() {defaultValue=1, group=00_Base, parameterName=Tread Overall Intensity, sortPriority=34})) {name=Tread, nodeColor=<rgba 1 0 0.8906 1>}
  n55 = TextureSampleParameter2D() {group=00_Base, parameterName=Normal, samplerType=SAMPLERTYPE_Normal, sortPriority=0, texture=&T_BlankNormal_B}
  NamedRerouteDeclaration(None=@n55.RGB) {name=Normal, nodeColor=<rgba 0 1 0.5313 1>}
  n56 = Panner(Coordinate=TextureCoordinate() {uTiling=0.5, vTiling=0.5}) {speedX=0.1, speedY=0.1}
  n57 = Panner() {speedX=-0.1, speedY=0.05}
  NamedRerouteDeclaration(None=Add(
    A=TextureSample(UVs=@n56) {samplerSource=SSM_Wrap_WorldGroupSettings, texture=&CloudNoise_01_uncompressed}.R
    B=TextureSample(UVs=@n57) {samplerSource=SSM_Wrap_WorldGroupSettings, texture=&CloudNoise_01_uncompressed}.R
   )) {name=Cloud Noise Tread, nodeColor=<rgba 0 0.3906 1 1>}
  NamedRerouteDeclaration(None=@n14.G) {name=G_Channel, nodeColor=<rgba 0.0497 0.8278 0 1>}
  NamedRerouteDeclaration(None=@n26.G) {name=VC_G, nodeColor=<rgba 0 1 0 1>}
  NamedRerouteDeclaration(None=@n26.B) {name=VC_B, nodeColor=<rgba 0 0 1 1>}
  NamedRerouteDeclaration(None=@n14.G) {name=Masks G, nodeColor=<rgba 0 1 0 1>}
  NamedRerouteDeclaration(None=@n12.R) {name=VFX R, nodeColor=<rgba 1 0 0 1>}
  NamedRerouteDeclaration(None=@n12.G) {name=VFX G, nodeColor=<rgba 0 1 0 1>}
  NamedRerouteDeclaration(None=@n12.B) {name=VFX B, nodeColor=<rgba 0 0 1 1>}
  NamedRerouteUsage()
  NamedRerouteUsage()
 ]
}
